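# Translucent, two-sided unlit material for stylized flames on Niagara mesh particles in the Castle AnglePatch Heart effect.
# Two panning cell-noise samples on 12-step-per-second time, offset by vertex red, distort a radial UV mask into flickering flame shapes.
# The depth-faded mask drives a hard-edged border-to-core color blend through emissive adjustments.
# World position offset combines vertex-color-controlled SplineThicken width with a nudge toward the camera.
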Material /Game/Environments/Helios/FX/Castle/Effects/AnglePatchHeart/M_Env_AnglePatch_Heart_Flames_Unlit
graph {
 properties {
  bAutomaticallySetUsageInEditor=false
  bUsedWithNiagaraMeshParticles=true
  blendMode=BLEND_Translucent
  shadingModel=MSM_Unlit
  twoSided=true
  usedShadingModels=Unlit
 }
 nodes [
  n1 = ScalarParameter() {defaultValue=9, parameterName=WidthBase}
  n2 = VertexColor()
  n3 = ScalarParameter() {defaultValue=6, parameterName=WidthTip}
  n4 = LinearInterpolate(
   A=MaterialFunctionCall(WidthBase=Multiply(A=@n1, B=@n2.B), WidthTip=Multiply(A=@n3, B=@n2.B)) {materialFunction=&SplineThicken}.WorldPositionOffset
   Alpha=VertexColor().G
  ) {constB=0}
  n5 = TextureCoordinate()
  n6 = Power(Base=OneMinus(None=ComponentMask(None=@n5) {g=true}))
  n7 = Panner(Coordinate=@n5) {speedY=0.7}
  n8 = Sine(None=Multiply(A=Panner(Coordinate=@n5) {speedY=0.4}) {constB=2})
  n9 = StaticComponentMaskParameter(None=Multiply(A=Sine(None=@n7), B=Multiply(A=@n8) {constB=0.4})) {defaultG=true}
  n10 = Multiply(A=Subtract(A=Multiply(A=@n9) {constB=2}), B=Constant3Vector() {constant=<rgba 0 1 1.1 0>})
  Add(A=@n4, B=Multiply(A=Multiply(A=@n6, B=@n10)) {constB=1.5})
  n11 = Add(A=TextureCoordinate(), B=AppendVector(A=VertexColor().R, B=Constant()))
  n12 = Time()
  n13 = Constant() {r=12}
  n14 = Divide(A=Floor(None=Multiply(A=@n12, B=@n13) {constB=12}), B=@n13) {constB=12}
  n15 = Panner(Coordinate=Multiply(A=@n11) {constB=0.3}, Time=@n14) {speedX=-0.1, speedY=0.3}
  n16 = Panner(Coordinate=Multiply(A=@n11) {constB=0.45}, Time=@n14) {speedX=-0.3, speedY=0.4}
  n17 = OneMinus(None=Max(
    A=TextureSample(UVs=@n15) {texture=/Game/Environments/Helios/FX/TheRaft/T_TheRaft_CellNoise}.R
    B=TextureSample(UVs=@n16) {texture=&T_TheRaft_CellNoise}.R
   ))
  n18 = AppendVector(A=Constant(), B=Multiply(A=Power(Base=@n17) {constExponent=3}) {constB=0.6})
  n19 = TextureCoordinate()
  n20 = SmoothStep(Value=OneMinus(None=ComponentMask(None=@n19) {g=true})) {constMax=0.3}
  n21 = Subtract(A=Add(A=TextureCoordinate(), B=Multiply(A=@n18, B=@n20)), B=Constant2Vector() {g=0.9, r=0.5})
  n22 = Power(Base=OneMinus(None=Length(None=@n21) {is=class MaterialExpressionLength})) {constExponent=5}
  n23 = Multiply(A=@n22, B=DepthFade() {fadeDistanceDefault=1.4})
  MaterialFunctionCall(Emissive=LinearInterpolate(
    A=VectorParameter() {group=Color, parameterName=BorderColor, sortPriority=1, defaultValue=<rgba 0 0 0 1>}.RGB
    B=VectorParameter() {group=Color, parameterName=CoreColor, sortPriority=2, defaultValue=<rgba 1 0.4466 0 1>}.RGB
    Alpha=SmoothStep(Value=@n23) {constMax=0.11, constMin=0.11}
   )) {materialFunction=&MF_FX_EmissiveAdjustments}
  n24 = Normalize(VectorInput=Subtract(A=ObjectPositionWS(), B=CameraPositionWS()))
  NamedRerouteDeclaration(None=Add(A=Multiply(A=@n24) {constB=-1}, B=@n4)) {name=WPO, nodeColor=<rgba 1 0 0.2812 1>}
  NamedRerouteUsage()
  Reroute()
  SmoothStep(Value=@n22) {constMax=0.092, constMin=0.09}
  TextureCoordinate()
 ]
}
